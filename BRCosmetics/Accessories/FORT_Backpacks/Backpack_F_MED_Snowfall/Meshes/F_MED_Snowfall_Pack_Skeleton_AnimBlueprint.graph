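# Animation Blueprint for the Snowfall backpack that poses its wings to match the player's movement state.
# The wingPosition state machine picks poses for base, jump, fall, targeting, crouch and nested dive, float and parachute states.
# Its cached pose gets speed-driven run/sprint additives, except while crouching or skydiving, then a pack bone offset.
# Wing Anim Dynamics and lean blendspaces exist but are unconnected to the output.
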AnimBlueprint /BRCosmetics/Accessories/FORT_Backpacks/Backpack_F_MED_Snowfall/Meshes/F_MED_Snowfall_Pack_Skeleton_AnimBlueprint
target_skeleton=/Game/Accessories/FORT_Backpacks/Backpack_F_MED_Snowfall/Meshes/F_MED_Snowfall_Pack_Skeleton
graphs [
 AnimGraph {nodes [
   n1 = Anim Dynamics - Bone: dyn_wingShoulder_lft(Alpha=1, ExternalForce="0,0,0") {
    alpha=0.5
    angularDampingOverride=0.9999
    angularSpringConstant=8000
    bAngularSpring=true
    bGravityOverrideInSimSpace=true
    bOverrideAngularDamping=true
    bUseGravityOverride=true
    simulationSpace=BoneRelative
    alphaBoolBlend {blendOption=Linear}
    boundBone {boneName=dyn_wingShoulder_lft}
    componentAppliedLinearAccClamp=<2000 2000 2000>
    componentLinearAccScale=<5 5 5>
    componentLinearVelScale=<5 5 5>
    gravityOverride {z=-980}
    physicsBodyDefinitions [{
      sphereCollisionRadius=0
      boundBone {boneName=dyn_wingShoulder_lft}
      boxExtents=<10 20 20>
      constraintSetup {angularTargetAxis=AxisY, angularLimitsMax=<0 3.5 2>, angularLimitsMin=<0 -3.5 -2>, angularTarget=<0 1 0>}
      localJointOffset=<-0.3 0 0>
     }]
   }
   n2 = Anim Dynamics - Bone: dyn_feather0_lft(ComponentPose=@n1, Alpha=1, ExternalForce="0,0,0") {
    alpha=0.5
    angularDampingOverride=0.9999
    angularSpringConstant=8000
    bAngularSpring=true
    bGravityOverrideInSimSpace=true
    bOverrideAngularDamping=true
    bUseGravityOverride=true
    simulationSpace=BoneRelative
    alphaBoolBlend {blendOption=Linear}
    boundBone {boneName=dyn_feather0_lft}
    componentAppliedLinearAccClamp=<2000 2000 2000>
    componentLinearAccScale=<5 5 5>
    componentLinearVelScale=<5 5 5>
    gravityOverride {z=-980}
    physicsBodyDefinitions [{
      sphereCollisionRadius=0
      boundBone {boneName=dyn_feather0_lft}
      boxExtents=<10 20 20>
      constraintSetup {angularTargetAxis=AxisY, angularLimitsMax=<0 2 20>, angularLimitsMin=<0 -2 -20>, angularTarget=<0 1 0>}
      localJointOffset=<-0.3 0 0>
     }]
   }
   n3 = Anim Dynamics - Bone: dyn_feather1_lft(ComponentPose=@n2, Alpha=1, ExternalForce="0,0,0") {
    alpha=0.5
    angularDampingOverride=0.9999
    angularSpringConstant=8000
    bAngularSpring=true
    bGravityOverrideInSimSpace=true
    bOverrideAngularDamping=true
    bUseGravityOverride=true
    simulationSpace=BoneRelative
    alphaBoolBlend {blendOption=Linear}
    boundBone {boneName=dyn_feather1_lft}
    componentAppliedLinearAccClamp=<2000 2000 2000>
    componentLinearAccScale=<5 5 5>
    componentLinearVelScale=<5 5 5>
    gravityOverride {z=-980}
    physicsBodyDefinitions [{
      sphereCollisionRadius=0
      boundBone {boneName=dyn_feather1_lft}
      boxExtents=<10 20 20>
      constraintSetup {angularTargetAxis=AxisY, angularLimitsMax=<0 2 20>, angularLimitsMin=<0 -2 -20>, angularTarget=<0 1 0>}
      localJointOffset=<-0.3 0 0>
     }]
   }
   n4 = Anim Dynamics - Bone: dyn_feather2_lft(ComponentPose=@n3, Alpha=1, ExternalForce="0,0,0") {
    alpha=0.5
    angularDampingOverride=0.9999
    angularSpringConstant=8000
    bAngularSpring=true
    bGravityOverrideInSimSpace=true
    bOverrideAngularDamping=true
    bUseGravityOverride=true
    simulationSpace=BoneRelative
    alphaBoolBlend {blendOption=Linear}
    boundBone {boneName=dyn_feather2_lft}
    componentAppliedLinearAccClamp=<2000 2000 2000>
    componentLinearAccScale=<5 5 5>
    componentLinearVelScale=<5 5 5>
    gravityOverride {z=-980}
    physicsBodyDefinitions [{
      sphereCollisionRadius=0
      boundBone {boneName=dyn_feather2_lft}
      boxExtents=<10 20 20>
      constraintSetup {angularTargetAxis=AxisY, angularLimitsMax=<0 2 20>, angularLimitsMin=<0 -2 -20>, angularTarget=<0 1 0>}
      localJointOffset=<-0.3 0 0>
     }]
   }
   n5 = Anim Dynamics - Bone: dyn_feather3_lft(ComponentPose=@n4, Alpha=1, ExternalForce="0,0,0") {
    alpha=0.5
    angularDampingOverride=0.9999
    angularSpringConstant=8000
    bAngularSpring=true
    bGravityOverrideInSimSpace=true
    bOverrideAngularDamping=true
    bUseGravityOverride=true
    simulationSpace=BoneRelative
    alphaBoolBlend {blendOption=Linear}
    boundBone {boneName=dyn_feather3_lft}
    componentAppliedLinearAccClamp=<2000 2000 2000>
    componentLinearAccScale=<5 5 5>
    componentLinearVelScale=<5 5 5>
    gravityOverride {z=-980}
    physicsBodyDefinitions [{
      sphereCollisionRadius=0
      boundBone {boneName=dyn_feather3_lft}
      boxExtents=<10 20 20>
      constraintSetup {angularTargetAxis=AxisY, angularLimitsMax=<0 2 20>, angularLimitsMin=<0 -2 -20>, angularTarget=<0 1 0>}
      localJointOffset=<-0.3 0 0>
     }]
   }
   n6 = Anim Dynamics - Bone: dyn_wing4_lft(ComponentPose=@n5, Alpha=1, ExternalForce="0,0,0") {
    alpha=0.5
    angularDampingOverride=0.9999
    angularSpringConstant=8000
    bAngularSpring=true
    bGravityOverrideInSimSpace=true
    bOverrideAngularDamping=true
    bUseGravityOverride=true
    simulationSpace=BoneRelative
    alphaBoolBlend {blendOption=Linear}
    boundBone {boneName=dyn_wing4_lft}
    componentAppliedLinearAccClamp=<2000 2000 2000>
    componentLinearAccScale=<5 5 5>
    componentLinearVelScale=<5 5 5>
    gravityOverride {z=-980}
    physicsBodyDefinitions [{
      sphereCollisionRadius=0
      boundBone {boneName=dyn_wing4_lft}
      boxExtents=<10 20 20>
      constraintSetup {angularTargetAxis=AxisY, angularLimitsMax=<0 2 20>, angularLimitsMin=<0 -2 -20>, angularTarget=<0 1 0>}
      localJointOffset=<-0.3 0 0>
     }]
   }
   n7 = Anim Dynamics - Bone: dyn_wingShoulder_rgt(ComponentPose=@n6, Alpha=1, ExternalForce="0,0,0") {
    alpha=0.5
    angularDampingOverride=0.9999
    angularSpringConstant=8000
    bAngularSpring=true
    bGravityOverrideInSimSpace=true
    bOverrideAngularDamping=true
    bUseGravityOverride=true
    simulationSpace=BoneRelative
    alphaBoolBlend {blendOption=Linear}
    boundBone {boneName=dyn_wingShoulder_rgt}
    componentAppliedLinearAccClamp=<2000 2000 2000>
    componentLinearAccScale=<5 5 5>
    componentLinearVelScale=<5 5 5>
    gravityOverride {z=-980}
    physicsBodyDefinitions [{
      sphereCollisionRadius=0
      boundBone {boneName=dyn_wingShoulder_rgt}
      boxExtents=<10 20 20>
      constraintSetup {angularTargetAxis=AxisY, angularLimitsMax=<0 3.5 2>, angularLimitsMin=<0 -3.5 -2>, angularTarget=<0 1 0>}
      localJointOffset=<-0.3 0 0>
     }]
   }
   n8 = Anim Dynamics - Bone: dyn_feather0_rgt(ComponentPose=@n7, Alpha=1, ExternalForce="0,0,0") {
    alpha=0.5
    angularDampingOverride=0.9999
    angularSpringConstant=8000
    bAngularSpring=true
    bGravityOverrideInSimSpace=true
    bOverrideAngularDamping=true
    bUseGravityOverride=true
    simulationSpace=BoneRelative
    alphaBoolBlend {blendOption=Linear}
    boundBone {boneName=dyn_feather0_rgt}
    componentAppliedLinearAccClamp=<2000 2000 2000>
    componentLinearAccScale=<5 5 5>
    componentLinearVelScale=<5 5 5>
    gravityOverride {z=-980}
    physicsBodyDefinitions [{
      sphereCollisionRadius=0
      boundBone {boneName=dyn_feather0_rgt}
      boxExtents=<10 20 20>
      constraintSetup {angularTargetAxis=AxisY, angularLimitsMax=<0 2 20>, angularLimitsMin=<0 -2 -20>, angularTarget=<0 1 0>}
      localJointOffset=<-0.3 0 0>
     }]
   }
   n9 = Anim Dynamics - Bone: dyn_feather1_rgt(ComponentPose=@n8, Alpha=1, ExternalForce="0,0,0") {
    alpha=0.5
    angularDampingOverride=0.9999
    angularSpringConstant=8000
    bAngularSpring=true
    bGravityOverrideInSimSpace=true
    bOverrideAngularDamping=true
    bUseGravityOverride=true
    simulationSpace=BoneRelative
    alphaBoolBlend {blendOption=Linear}
    boundBone {boneName=dyn_feather1_rgt}
    componentAppliedLinearAccClamp=<2000 2000 2000>
    componentLinearAccScale=<5 5 5>
    componentLinearVelScale=<5 5 5>
    gravityOverride {z=-980}
    physicsBodyDefinitions [{
      sphereCollisionRadius=0
      boundBone {boneName=dyn_feather1_rgt}
      boxExtents=<10 20 20>
      constraintSetup {angularTargetAxis=AxisY, angularLimitsMax=<0 2 20>, angularLimitsMin=<0 -2 -20>, angularTarget=<0 1 0>}
      localJointOffset=<-0.3 0 0>
     }]
   }
   n10 = Anim Dynamics - Bone: dyn_feather2_rgt(ComponentPose=@n9, Alpha=1, ExternalForce="0,0,0") {
    alpha=0.5
    angularDampingOverride=0.9999
    angularSpringConstant=8000
    bAngularSpring=true
    bGravityOverrideInSimSpace=true
    bOverrideAngularDamping=true
    bUseGravityOverride=true
    simulationSpace=BoneRelative
    alphaBoolBlend {blendOption=Linear}
    boundBone {boneName=dyn_feather2_rgt}
    componentAppliedLinearAccClamp=<2000 2000 2000>
    componentLinearAccScale=<5 5 5>
    componentLinearVelScale=<5 5 5>
    gravityOverride {z=-980}
    physicsBodyDefinitions [{
      sphereCollisionRadius=0
      boundBone {boneName=dyn_feather2_rgt}
      boxExtents=<10 20 20>
      constraintSetup {angularTargetAxis=AxisY, angularLimitsMax=<0 2 20>, angularLimitsMin=<0 -2 -20>, angularTarget=<0 1 0>}
      localJointOffset=<-0.3 0 0>
     }]
   }
   n11 = Anim Dynamics - Bone: dyn_feather3_rgt(ComponentPose=@n10, Alpha=1, ExternalForce="0,0,0") {
    alpha=0.5
    angularDampingOverride=0.9999
    angularSpringConstant=8000
    bAngularSpring=true
    bGravityOverrideInSimSpace=true
    bOverrideAngularDamping=true
    bUseGravityOverride=true
    simulationSpace=BoneRelative
    alphaBoolBlend {blendOption=Linear}
    boundBone {boneName=dyn_feather3_rgt}
    componentAppliedLinearAccClamp=<2000 2000 2000>
    componentLinearAccScale=<5 5 5>
    componentLinearVelScale=<5 5 5>
    gravityOverride {z=-980}
    physicsBodyDefinitions [{
      sphereCollisionRadius=0
      boundBone {boneName=dyn_feather3_rgt}
      boxExtents=<10 20 20>
      constraintSetup {angularTargetAxis=AxisY, angularLimitsMax=<0 2 20>, angularLimitsMin=<0 -2 -20>, angularTarget=<0 1 0>}
      localJointOffset=<-0.3 0 0>
     }]
   }
   Anim Dynamics - Bone: dyn_wing4_rgt(ComponentPose=@n11, Alpha=1, ExternalForce="0,0,0") {
    alpha=0.5
    angularDampingOverride=0.9999
    angularSpringConstant=8000
    bAngularSpring=true
    bGravityOverrideInSimSpace=true
    bOverrideAngularDamping=true
    bUseGravityOverride=true
    simulationSpace=BoneRelative
    alphaBoolBlend {blendOption=Linear}
    boundBone {boneName=dyn_wing4_rgt}
    componentAppliedLinearAccClamp=<2000 2000 2000>
    componentLinearAccScale=<5 5 5>
    componentLinearVelScale=<5 5 5>
    gravityOverride {z=-980}
    physicsBodyDefinitions [{
      sphereCollisionRadius=0
      boundBone {boneName=dyn_wing4_rgt}
      boxExtents=<10 20 20>
      constraintSetup {angularTargetAxis=AxisY, angularLimitsMax=<0 2 20>, angularLimitsMin=<0 -2 -20>, angularTarget=<0 1 0>}
      localJointOffset=<-0.3 0 0>
     }]
   }
   Component To Local()
   Local To Component()
   "Transform (Modify) Bone - Bone: pack"(Translation="0,0,0", Rotation="0,0,0", Scale="0.001,0.001,0.001") {
    scaleMode=BMM_Additive
    scaleSpace=BCS_BoneSpace
    alphaBoolBlend {blendOption=Linear}
    boneToModify {boneName=pack}
    scale=<0.001 0.001 0.001>
   }
   n12 = F_MED_Snowfall_Backpack_Poses_Anim_PoseAsset_Add : Run!(PoseWeight:float=1) -> (Pose:Pose Link Structure) {
    is=title AnimGraphNode_PoseByName
    poseAsset=/BRCosmetics/Accessories/FORT_Backpacks/Backpack_F_MED_Snowfall/Animations/F_MED_Snowfall_Backpack_Poses_Anim_PoseAsset_Add
    poseName=Run
   }
   n13 = F_MED_Snowfall_Backpack_Poses_Anim_PoseAsset_Add : Sprint!(PoseWeight:float=1) -> (Pose:Pose Link Structure) {
    is=title AnimGraphNode_PoseByName
    poseAsset=&F_MED_Snowfall_Backpack_Poses_Anim_PoseAsset_Add
    poseName=Sprint
   }
   n14 = Get SprintSpeedRemap() -> (SprintSpeedRemap:float)
   n15 = Two Way Blend(
    A=Apply Additive(Base=Use cached pose 'preSprint'(), Additive=@n13, Alpha=@n14) {alphaBoolBlend {blendOption=Linear}, alphaScaleBiasClamp {bInterpResult=true, interpSpeedDecreasing=5}}
    B=Apply Additive(Base=Use cached pose 'preSprint'(), Additive=@n12, Alpha=Get RunSpeedRemap() -> (RunSpeedRemap:float)) {alphaBoolBlend {blendOption=Linear}, alphaScaleBiasClamp {bInterpResult=true, interpSpeedDecreasing=5}}
    Alpha=@n14
   ) {blendNode {alphaBoolBlend {blendOption=Linear}, alphaScaleBiasClamp {bInterpResult=true, bias=1, interpSpeedDecreasing=5, scale=-1}}}
   n16 = Blend Poses by bool(BlendPose_0=Use cached pose 'preSprint'(), BlendPose_1=@n15, BlendTime_0=0.3, BlendTime_1=0.3, bActiveValue=Get bIsCrouching()) {blendType=Linear}
   n17 = Use cached pose 'preSprint'()
   n18 = Get bIsSkyDiving()
   n19 = "Transform (Modify) Bone - Bone: pack"(
    ComponentPose=Local To Component(LocalPose=Blend Poses by bool(BlendPose_0=@n17, BlendPose_1=@n16, BlendTime_0=0.3, BlendTime_1=0.3, bActiveValue=@n18) {blendType=Linear})
    Translation="0.000,2,0.000"
    Rotation="0,0,0"
   ) {
    scaleMode=BMM_Additive
    scaleSpace=BCS_BoneSpace
    translationMode=BMM_Additive
    translationSpace=BCS_BoneSpace
    alphaBoolBlend {blendOption=Linear}
    boneToModify {boneName=pack}
    translation {y=2}
   }
   Output Pose(Result=Component To Local(ComponentPose=@n19)) {layerGroup=None}
   AnimGraphNode_StateMachine_0 = wingPosition() {is=title AnimGraphNode_StateMachine}
   Save cached pose 'preSprint'(Pose=Component To Local(ComponentPose=Local To Component(LocalPose=@AnimGraphNode_StateMachine_0))) {cacheName=preSprint}
  ]}
 wingPosition {
  entry=Base
  path=AnimGraph.AnimGraphNode_StateMachine_0.wingPosition
  ids=id
  nodes [
   AnimStateNode_0 = Base() {
    is=title AnimStateNode
    stateEntered {guid=#1}
    stateFullyBlended {guid=#2}
    stateLeft {guid=#3}
   }
   AnimStateNode_1 = Jumping() {
    is=title AnimStateNode
    stateEntered {guid=#4}
    stateFullyBlended {guid=#5}
    stateLeft {guid=#6}
   }
   AnimStateNode_2 = Falling() {
    is=title AnimStateNode
    stateEntered {guid=#7}
    stateFullyBlended {guid=#8}
    stateLeft {guid=#9}
   }
   AnimStateNode_3 = Targeting() {
    is=title AnimStateNode
    stateEntered {guid=#10}
    stateFullyBlended {guid=#11}
    stateLeft {guid=#12}
   }
   AnimStateNode_5 = SkydiveMode() {
    is=title AnimStateNode
    stateEntered {guid=#13}
    stateFullyBlended {guid=#14}
    stateLeft {guid=#15}
   }
   AnimStateNode_9 = Crouching() {
    is=title AnimStateNode
    stateEntered {guid=#16}
    stateFullyBlended {guid=#17}
    stateLeft {guid=#18}
   }
   AnimStateTransitionNode_0 = AnimStateTransitionNode() {
    blendMode=Custom
    customBlendCurve=/Game/Accessories/FORT_Backpacks/M_MED_Demon_Pack/Animation/JumpTransitionBlendCurve
    priorityOrder=2
    transitionEnd {guid=#19}
    transitionInterrupt {guid=#20}
    transitionStart {guid=#21}
   }
   AnimStateTransitionNode_1 = AnimStateTransitionNode() {
    blendMode=Custom
    crossfadeDuration=0.5
    customBlendCurve=/Game/Accessories/FORT_Backpacks/M_MED_Demon_Pack/Animation/JumpTransitionBlendOutCurve
    transitionEnd {guid=#22}
    transitionInterrupt {guid=#23}
    transitionStart {guid=#24}
   }
   AnimStateTransitionNode_11 = AnimStateTransitionNode() {
    transitionEnd {guid=#25}
    transitionInterrupt {guid=#26}
    transitionStart {guid=#27}
   }
   AnimStateTransitionNode_12 = AnimStateTransitionNode() {
    transitionEnd {guid=#28}
    transitionInterrupt {guid=#29}
    transitionStart {guid=#30}
   }
   AnimStateTransitionNode_13 = AnimStateTransitionNode() {
    transitionEnd {guid=#31}
    transitionInterrupt {guid=#32}
    transitionStart {guid=#33}
   }
   AnimStateTransitionNode_15 = AnimStateTransitionNode() {
    transitionEnd {guid=#34}
    transitionInterrupt {guid=#35}
    transitionStart {guid=#36}
   }
   AnimStateTransitionNode_2 = AnimStateTransitionNode()
   AnimStateTransitionNode_3 = AnimStateTransitionNode() {
    blendMode=Custom
    customBlendCurve=&JumpTransitionBlendOutCurve
    transitionEnd {guid=#37}
    transitionInterrupt {guid=#38}
    transitionStart {guid=#39}
   }
   AnimStateTransitionNode_4 = AnimStateTransitionNode() {
    blendMode=Custom
    customBlendCurve=&JumpTransitionBlendCurve
    transitionEnd {guid=#40}
    transitionInterrupt {guid=#41}
    transitionStart {guid=#42}
   }
   AnimStateTransitionNode_5 = AnimStateTransitionNode() {
    transitionEnd {guid=#43}
    transitionInterrupt {guid=#44}
    transitionStart {guid=#45}
   }
   AnimStateTransitionNode_6 = AnimStateTransitionNode() {
    crossfadeDuration=0.25
    transitionEnd {guid=#46}
    transitionInterrupt {guid=#47}
    transitionStart {guid=#48}
   }
   AnimStateTransitionNode_7 = AnimStateTransitionNode() {
    crossfadeDuration=0.25
    transitionEnd {guid=#49}
    transitionInterrupt {guid=#50}
    transitionStart {guid=#51}
   }
  ]
  links [
   {
    dst=Crouching
    rule=AnimGraph.AnimGraphNode_StateMachine_0.wingPosition.AnimStateTransitionNode_12.AnimationTransitionGraph_1
    src=Base
    via=AnimStateTransitionNode_12
   }
   {
    dst=Falling
    rule=AnimGraph.AnimGraphNode_StateMachine_0.wingPosition.AnimStateTransitionNode_4.AnimationTransitionGraph_1
    src=Base
    via=AnimStateTransitionNode_4
   }
   {
    dst=Jumping
    rule=AnimGraph.AnimGraphNode_StateMachine_0.wingPosition.AnimStateTransitionNode_0.AnimationTransitionGraph_1
    src=Base
    via=AnimStateTransitionNode_0
   }
   {
    dst=SkydiveMode
    rule=AnimGraph.AnimGraphNode_StateMachine_0.wingPosition.AnimStateTransitionNode_15.AnimationTransitionGraph_1
    src=Base
    via=AnimStateTransitionNode_15
   }
   {
    dst=Targeting
    rule=AnimGraph.AnimGraphNode_StateMachine_0.wingPosition.AnimStateTransitionNode_6.AnimationTransitionGraph_1
    src=Base
    via=AnimStateTransitionNode_6
   }
   {
    dst=Base
    rule=AnimGraph.AnimGraphNode_StateMachine_0.wingPosition.AnimStateTransitionNode_13.AnimationTransitionGraph_1
    src=Crouching
    via=AnimStateTransitionNode_13
   }
   {
    dst=Base
    rule=AnimGraph.AnimGraphNode_StateMachine_0.wingPosition.AnimStateTransitionNode_3.AnimationTransitionGraph_1
    src=Falling
    via=AnimStateTransitionNode_3
   }
   {
    dst=Base
    rule=AnimGraph.AnimGraphNode_StateMachine_0.wingPosition.AnimStateTransitionNode_1.AnimationTransitionGraph_1
    src=Jumping
    via=AnimStateTransitionNode_1
   }
   {
    dst=Falling
    rule=AnimGraph.AnimGraphNode_StateMachine_0.wingPosition.AnimStateTransitionNode_2.AnimationTransitionGraph_1
    src=Jumping
    via=AnimStateTransitionNode_2
   }
   {
    dst=SkydiveMode
    rule=AnimGraph.AnimGraphNode_StateMachine_0.wingPosition.AnimStateTransitionNode_5.AnimationTransitionGraph_1
    src=Jumping
    via=AnimStateTransitionNode_5
   }
   {
    dst=Base
    rule=AnimGraph.AnimGraphNode_StateMachine_0.wingPosition.AnimStateTransitionNode_11.AnimationTransitionGraph_1
    src=SkydiveMode
    via=AnimStateTransitionNode_11
   }
   {
    dst=Base
    rule=AnimGraph.AnimGraphNode_StateMachine_0.wingPosition.AnimStateTransitionNode_7.AnimationTransitionGraph_1
    src=Targeting
    via=AnimStateTransitionNode_7
   }
  ]
 }
 Base {
  path=AnimGraph.AnimGraphNode_StateMachine_0.wingPosition.AnimStateNode_0.Base
  nodes [
   n1 = F_MED_Snowfall_Backpack_Poses_Anim_PoseAsset : Base!(PoseWeight:float=1) -> (Pose:Pose Link Structure) {
    is=title AnimGraphNode_PoseByName
    poseAsset=/BRCosmetics/Accessories/FORT_Backpacks/Backpack_F_MED_Snowfall/Animations/F_MED_Snowfall_Backpack_Poses_Anim_PoseAsset
    poseName=Base
   }
   Output Animation Pose(Result=Apply Additive(Base=@n1) {alphaBoolBlend {blendOption=Linear}})
  ]
 }
 Jumping {
  path=AnimGraph.AnimGraphNode_StateMachine_0.wingPosition.AnimStateNode_1.Jumping
  nodes [
   n1 = F_MED_Snowfall_Backpack_Poses_Anim_PoseAsset : Jump!(PoseWeight:float=1) -> (Pose:Pose Link Structure) {
    is=title AnimGraphNode_PoseByName
    poseAsset=&F_MED_Snowfall_Backpack_Poses_Anim_PoseAsset
    poseName=Jump
   }
   Output Animation Pose(Result=Apply Additive(Base=@n1) {alphaBoolBlend {blendOption=Linear}})
  ]
 }
 Falling {
  path=AnimGraph.AnimGraphNode_StateMachine_0.wingPosition.AnimStateNode_2.Falling
  nodes [
   Blendspace Player 'F_MED_Snowfall_Backpack_Wings_Leans_Ground_Blend_Space'(
    X=Get PlayerSidewaysAcceleration() -> (PlayerSidewaysAcceleration:float)
    Y=Get PlayerForwardAcceleration() -> (PlayerForwardAcceleration:float)
   ) {blendSpace=/Game/Accessories/FORT_Backpacks/Backpack_F_MED_Snowfall/Animations/F_MED_Snowfall_Backpack_Wings_Leans_Ground_Blend_Space}
   n1 = F_MED_Snowfall_Backpack_Poses_Anim_PoseAsset : Jump(PoseWeight=1) {
    poseAsset=&F_MED_Snowfall_Backpack_Poses_Anim_PoseAsset
    poseName=Jump
   }
   Output Animation Pose(Result=Apply Additive(Base=@n1) {alphaBoolBlend {blendOption=Linear}})
  ]
 }
 Targeting {
  path=AnimGraph.AnimGraphNode_StateMachine_0.wingPosition.AnimStateNode_3.Targeting
  nodes [
   n1 = F_MED_Snowfall_Backpack_Poses_Anim_PoseAsset : Targeting!(PoseWeight:float=1) -> (Pose:Pose Link Structure) {
    is=title AnimGraphNode_PoseByName
    poseAsset=&F_MED_Snowfall_Backpack_Poses_Anim_PoseAsset
    poseName=Targeting
   }
   Output Animation Pose(Result=@n1)
  ]
 }
 SkydiveMode {
  path=AnimGraph.AnimGraphNode_StateMachine_0.wingPosition.AnimStateNode_5.SkydiveMode
  nodes [
   AnimGraphNode_StateMachine_0 = Skydiving State() {is=title AnimGraphNode_StateMachine}
   Output Animation Pose(Result=@AnimGraphNode_StateMachine_0)
  ]
 }
 Skydiving State {
  entry=Parachuting
  path=AnimGraph.AnimGraphNode_StateMachine_0.wingPosition.AnimStateNode_5.SkydiveMode.AnimGraphNode_StateMachine_0.Skydiving State
  ids=id
  nodes [
   AnimStateNode_1 = Diving() {is=title AnimStateNode}
   AnimStateNode_2 = Floating() {is=title AnimStateNode}
   AnimStateNode_4 = Parachuting() {is=title AnimStateNode}
   AnimStateTransitionNode_1 = AnimStateTransitionNode() {crossfadeDuration=0.5}
   AnimStateTransitionNode_10 = AnimStateTransitionNode() {crossfadeDuration=0.976}
   AnimStateTransitionNode_12 = AnimStateTransitionNode()
   AnimStateTransitionNode_13 = AnimStateTransitionNode()
   AnimStateTransitionNode_3 = AnimStateTransitionNode()
   AnimStateTransitionNode_5 = AnimStateTransitionNode() {crossfadeDuration=0.5}
  ]
  links [
   {
    dst=Floating
    rule=AnimGraph.AnimGraphNode_StateMachine_0.wingPosition.AnimStateNode_5.SkydiveMode.AnimGraphNode_StateMachine_0.Skydiving State.AnimStateTransitionNode_1.AnimationTransitionGraph_0
    src=Diving
    via=AnimStateTransitionNode_1
   }
   {
    dst=Parachuting
    rule=AnimGraph.AnimGraphNode_StateMachine_0.wingPosition.AnimStateNode_5.SkydiveMode.AnimGraphNode_StateMachine_0.Skydiving State.AnimStateTransitionNode_10.AnimationTransitionGraph_1
    src=Diving
    via=AnimStateTransitionNode_10
   }
   {
    dst=Diving
    rule=AnimGraph.AnimGraphNode_StateMachine_0.wingPosition.AnimStateNode_5.SkydiveMode.AnimGraphNode_StateMachine_0.Skydiving State.AnimStateTransitionNode_5.AnimationTransitionGraph_0
    src=Floating
    via=AnimStateTransitionNode_5
   }
   {
    dst=Parachuting
    rule=AnimGraph.AnimGraphNode_StateMachine_0.wingPosition.AnimStateNode_5.SkydiveMode.AnimGraphNode_StateMachine_0.Skydiving State.AnimStateTransitionNode_3.AnimationTransitionGraph_1
    src=Floating
    via=AnimStateTransitionNode_3
   }
   {
    dst=Diving
    rule=AnimGraph.AnimGraphNode_StateMachine_0.wingPosition.AnimStateNode_5.SkydiveMode.AnimGraphNode_StateMachine_0.Skydiving State.AnimStateTransitionNode_13.AnimationTransitionGraph_1
    src=Parachuting
    via=AnimStateTransitionNode_13
   }
   {
    dst=Floating
    rule=AnimGraph.AnimGraphNode_StateMachine_0.wingPosition.AnimStateNode_5.SkydiveMode.AnimGraphNode_StateMachine_0.Skydiving State.AnimStateTransitionNode_12.AnimationTransitionGraph_1
    src=Parachuting
    via=AnimStateTransitionNode_12
   }
  ]
 }
 Diving {
  path=AnimGraph.AnimGraphNode_StateMachine_0.wingPosition.AnimStateNode_5.SkydiveMode.AnimGraphNode_StateMachine_0.Skydiving State.AnimStateNode_1.Diving
  nodes [
   Apply Additive() {alphaBoolBlend {blendOption=Linear}}
   Blendspace Player 'F_MED_Snowfall_Backpack_Wings_Leans_Air_Blend_Space'(X=Get PlayerSidewaysAcceleration(), Y=Get PlayerForwardAcceleration()) {blendSpace=/Game/Accessories/FORT_Backpacks/Backpack_F_MED_Snowfall/Animations/F_MED_Snowfall_Backpack_Wings_Leans_Air_Blend_Space}
   n1 = F_MED_Snowfall_Backpack_Poses_Anim_PoseAsset : Dive!(PoseWeight:float=1) -> (Pose:Pose Link Structure) {
    is=title AnimGraphNode_PoseByName
    poseAsset=&F_MED_Snowfall_Backpack_Poses_Anim_PoseAsset
    poseName=Dive
   }
   Output Animation Pose(Result=@n1)
  ]
 }
 Floating {
  path=AnimGraph.AnimGraphNode_StateMachine_0.wingPosition.AnimStateNode_5.SkydiveMode.AnimGraphNode_StateMachine_0.Skydiving State.AnimStateNode_2.Floating
  nodes [
   Blendspace Player 'F_MED_Snowfall_Backpack_Wings_Leans_Air_Blend_Space'(X=Get PlayerSidewaysAcceleration(), Y=Get PlayerForwardAcceleration()) {blendSpace=&F_MED_Snowfall_Backpack_Wings_Leans_Air_Blend_Space}
   n1 = F_MED_Snowfall_Backpack_Poses_Anim_PoseAsset : Float!(PoseWeight:float=1) -> (Pose:Pose Link Structure) {
    is=title AnimGraphNode_PoseByName
    poseAsset=&F_MED_Snowfall_Backpack_Poses_Anim_PoseAsset
    poseName=Float
   }
   Output Animation Pose(Result=Apply Additive(Base=@n1) {alphaBoolBlend {blendOption=Linear}})
  ]
 }
 Parachuting {
  path=AnimGraph.AnimGraphNode_StateMachine_0.wingPosition.AnimStateNode_5.SkydiveMode.AnimGraphNode_StateMachine_0.Skydiving State.AnimStateNode_4.Parachuting
  nodes [
   Blendspace Player 'F_MED_Snowfall_Backpack_Wings_Leans_Parachuting_Blend_Space'(X=Get PlayerSidewaysAcceleration(), Y=Get PlayerForwardAcceleration()) {blendSpace=/Game/Accessories/FORT_Backpacks/Backpack_F_MED_Snowfall/Animations/F_MED_Snowfall_Backpack_Wings_Leans_Parachuting_Blend_Space}
   n1 = F_MED_Snowfall_Backpack_Poses_Anim_PoseAsset : Float(PoseWeight=1) {
    poseAsset=&F_MED_Snowfall_Backpack_Poses_Anim_PoseAsset
    poseName=Float
   }
   Output Animation Pose(Result=Apply Additive(Base=@n1) {alphaBoolBlend {blendOption=Linear}})
  ]
 }
 AnimationTransitionGraph_0 {
  path=AnimGraph.AnimGraphNode_StateMachine_0.wingPosition.AnimStateNode_5.SkydiveMode.AnimGraphNode_StateMachine_0.Skydiving State.AnimStateTransitionNode_1.AnimationTransitionGraph_0
  nodes [Result(bCanEnterTransition=NOT Boolean(A=Get bIsSkydiveDiveMode()))]
 }
 AnimationTransitionGraph_1 {
  path=AnimGraph.AnimGraphNode_StateMachine_0.wingPosition.AnimStateNode_5.SkydiveMode.AnimGraphNode_StateMachine_0.Skydiving State.AnimStateTransitionNode_10.AnimationTransitionGraph_1
  nodes [Result(bCanEnterTransition=Get bIsParachuteOpened())]
 }
 AnimationTransitionGraph_1 {
  path=AnimGraph.AnimGraphNode_StateMachine_0.wingPosition.AnimStateNode_5.SkydiveMode.AnimGraphNode_StateMachine_0.Skydiving State.AnimStateTransitionNode_12.AnimationTransitionGraph_1
  nodes [Result(bCanEnterTransition=NOT Boolean(A=Get bIsParachuteOpened()))]
 }
 AnimationTransitionGraph_1 {
  path=AnimGraph.AnimGraphNode_StateMachine_0.wingPosition.AnimStateNode_5.SkydiveMode.AnimGraphNode_StateMachine_0.Skydiving State.AnimStateTransitionNode_13.AnimationTransitionGraph_1
  nodes [Result(bCanEnterTransition=NOT Boolean(A=Get bIsParachuteOpened()))]
 }
 AnimationTransitionGraph_1 {
  path=AnimGraph.AnimGraphNode_StateMachine_0.wingPosition.AnimStateNode_5.SkydiveMode.AnimGraphNode_StateMachine_0.Skydiving State.AnimStateTransitionNode_3.AnimationTransitionGraph_1
  nodes [Result(bCanEnterTransition=Get bIsParachuteOpened())]
 }
 AnimationTransitionGraph_0 {
  path=AnimGraph.AnimGraphNode_StateMachine_0.wingPosition.AnimStateNode_5.SkydiveMode.AnimGraphNode_StateMachine_0.Skydiving State.AnimStateTransitionNode_5.AnimationTransitionGraph_0
  nodes [Result(bCanEnterTransition=Get bIsSkydiveDiveMode())]
 }
 Crouching {
  path=AnimGraph.AnimGraphNode_StateMachine_0.wingPosition.AnimStateNode_9.Crouching
  nodes [
   n1 = F_MED_Snowfall_Backpack_Poses_Anim_PoseAsset : Crouch!(PoseWeight:float=1) -> (Pose:Pose Link Structure) {
    is=title AnimGraphNode_PoseByName
    poseAsset=&F_MED_Snowfall_Backpack_Poses_Anim_PoseAsset
    poseName=Crouch
   }
   Output Animation Pose(Result=Apply Additive(Base=@n1) {alphaBoolBlend {blendOption=Linear}})
  ]
 }
 AnimationTransitionGraph_1 {
  path=AnimGraph.AnimGraphNode_StateMachine_0.wingPosition.AnimStateTransitionNode_0.AnimationTransitionGraph_1
  nodes [Result(bCanEnterTransition=Get bIsJumping())]
 }
 AnimationTransitionGraph_1 {
  path=AnimGraph.AnimGraphNode_StateMachine_0.wingPosition.AnimStateTransitionNode_1.AnimationTransitionGraph_1
  nodes [
   Result(bCanEnterTransition=NOT Boolean(A=Get bIsJumping()))
   Get bShouldPredictLanding() -> (bShouldPredictLanding:bool)
   Get bLandingPredicted() -> (bLandingPredicted:bool)
   Get bDebugLandPrediction() -> (bDebugLandPrediction:bool)
   Get bIsFalling()
  ]
 }
 AnimationTransitionGraph_1 {
  path=AnimGraph.AnimGraphNode_StateMachine_0.wingPosition.AnimStateTransitionNode_11.AnimationTransitionGraph_1
  nodes [Result(bCanEnterTransition=NOT Boolean(A=Get bIsSkyDiving()))]
 }
 AnimationTransitionGraph_1 {
  path=AnimGraph.AnimGraphNode_StateMachine_0.wingPosition.AnimStateTransitionNode_12.AnimationTransitionGraph_1
  nodes [Result(bCanEnterTransition=Get bIsCrouching())]
 }
 AnimationTransitionGraph_1 {
  path=AnimGraph.AnimGraphNode_StateMachine_0.wingPosition.AnimStateTransitionNode_13.AnimationTransitionGraph_1
  nodes [Result(bCanEnterTransition=NOT Boolean(A=Get bIsCrouching()))]
 }
 AnimationTransitionGraph_1 {
  path=AnimGraph.AnimGraphNode_StateMachine_0.wingPosition.AnimStateTransitionNode_15.AnimationTransitionGraph_1
  nodes [Result(bCanEnterTransition=Get bIsSkyDiving())]
 }
 AnimationTransitionGraph_1 {
  path=AnimGraph.AnimGraphNode_StateMachine_0.wingPosition.AnimStateTransitionNode_2.AnimationTransitionGraph_1
  nodes [Result(bCanEnterTransition=Get bIsFalling())]
 }
 AnimationTransitionGraph_1 {
  path=AnimGraph.AnimGraphNode_StateMachine_0.wingPosition.AnimStateTransitionNode_3.AnimationTransitionGraph_1
  nodes [Result(bCanEnterTransition=NOT Boolean(A=Get bIsFalling()))]
 }
 AnimationTransitionGraph_1 {
  path=AnimGraph.AnimGraphNode_StateMachine_0.wingPosition.AnimStateTransitionNode_4.AnimationTransitionGraph_1
  nodes [Result(bCanEnterTransition=Get bIsFalling())]
 }
 AnimationTransitionGraph_1 {
  path=AnimGraph.AnimGraphNode_StateMachine_0.wingPosition.AnimStateTransitionNode_5.AnimationTransitionGraph_1
  nodes [Result(bCanEnterTransition=Get bIsSkyDiving())]
 }
 AnimationTransitionGraph_1 {
  path=AnimGraph.AnimGraphNode_StateMachine_0.wingPosition.AnimStateTransitionNode_6.AnimationTransitionGraph_1
  nodes [Result(bCanEnterTransition=Get bIsTargeting())]
 }
 AnimationTransitionGraph_1 {
  path=AnimGraph.AnimGraphNode_StateMachine_0.wingPosition.AnimStateTransitionNode_7.AnimationTransitionGraph_1
  nodes [Result(bCanEnterTransition=NOT Boolean(A=Get bIsTargeting()))]
 }
 EventGraph {nodes [
   TryGetPawnOwner()
   Event BlueprintUpdateAnimation()
  ]}
]
variables []
